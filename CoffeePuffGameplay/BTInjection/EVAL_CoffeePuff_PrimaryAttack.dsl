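; CoffeePuff primary attack evaluator for behavior tree injection that blocks the attack while the AI's pawn is swimming.
; On enter, it casts the tree component's owner to a controller and caches its controlled Fort pawn.
; Starting and ongoing condition checks both keep the parent evaluator's result and add a not-swimming requirement.
; Swimming counts as either water-body surface swimming or the character movement component reporting swimming.
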(event EventOnEnter
  (params (OwnerComp :type Behavior%20Tree%20Component%20Object%20Reference))
  (bind _ascontroller (Utilities|Casting|CastToController (Components|GetOwner OwnerComp)))
  (bind _asfort_pawn (Utilities|Casting|CastToFortPawn (Pawn|GetControlledPawn _ascontroller)))
  (Variables|Default|SetPawn _asfort_pawn))

(fn EvaluateOngoingConditions
  (params (OwnerComp :type Behavior%20Tree%20Component%20Object%20Reference))
  (bind _returnvalue (|CallParentFunction OwnerComp))
  (return (and _returnvalue (not (CallFunction|IsSwimming)))))

(fn EvaluateStartingConditions
  (params (OwnerComp :type Behavior%20Tree%20Component%20Object%20Reference))
  (bind _returnvalue (|CallParentFunction OwnerComp))
  (return (and _returnvalue (not (CallFunction|IsSwimming)))))

(fn IsSwimming
  (return (or (WaterBody|IsSurfaceSwimming (Variables|Default|GetPawn)) (AI|Components|NavMovement|IsSwimming (Variables|/Script/Engine.Character|GetCharacterMovement (Variables|Default|GetPawn)) :!class /Script/Engine.NavMovementComponent))))
